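; Looping gameplay cue for the Jump Boots exotic that drives jump and landing effects, with the truncated raw magnitude as the stage number.
; On application, it waits for Body and Gameplay cosmetic parts, spawning jump Niagara, setting timed emissive material parameters and playing a stage-indexed sound on an audio component attached to the body.
; On removal, it spawns landing Niagara flagged as not a valid landing, plus a stage-indexed landing sound.
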(event EventBeginPlay)

(event Collision|EventActorBeginOverlap
  (params (OtherActor :type Actor%20Object%20Reference)))

(event EventOnApplication(EitherVFX-preferthisone)
  (params (MyTarget :type Actor%20Object%20Reference) (Parameters :type Gameplay%20Cue%20Parameters%20Structure%20%28by%20ref%29) (ParticleComponents :type Array%20of%20FXSystem%20Component%20Object%20References) (AudioComponents :type Array%20of%20Audio%20Component%20Object%20References) (BurstCameraShakeInstance :type Camera%20Shake%20Base%20Object%20Reference) (BurstDecalInstance :type Decal%20Actor%20Object%20Reference))
  (bind (_normalizedmagnitude _rawmagnitude) (Ability|GameplayCue|BreakGameplayCueParameters Parameters))
  (bind _returnvalue (Math|Float|Truncate _rawmagnitude))
  (bind _jumpboots_jump (Variables|Default|GetJumpBoots_Jump))
  (bind _asplayer_pawn_athena (Utilities|Casting|CastToPlayerPawn_Athena MyTarget))
  (bind (_pawn _part _meshpart) (Cosmetics|WaitForCosmeticSpecificPartApplied 0 0 true "Body")
    (:OnPartApplied
      (bind _returnvalue_2 (Niagara|SpawnSystematLocation 0 "/PrimalGameplay/Items/JumpBoots/Assets/Effects/NS_JumpBoots_OnJump.NS_JumpBoots_OnJump" (Transformation|GetSocketLocation _meshpart "root")))
      (Effects|Components|ParticleSystem|SetIntParameter _returnvalue_2 "StageNum" _returnvalue)
      ;[1 1 1] Material Emissives
      (bind (_pawn_2 _part_2 _meshpart_2) (Cosmetics|WaitForCosmeticSpecificPartApplied 0 _asplayer_pawn_athena true "Gameplay")
        (:OnPartApplied
          (Rendering|Material|SetScalarParameterValueonMaterials _meshpart_2 "IsLanding" 0.0)
          (Rendering|Material|SetScalarParameterValueonMaterials _meshpart_2 "StageNum" (Math|Conversions|ToFloat(Integer) _returnvalue))
          (Rendering|Material|SetScalarParameterValueonMaterials _meshpart_2 "GameTimeFromBP" (Utilities|Time|GetGameTimeinSeconds _meshpart_2))
          ;[1 1 0] WIP jump audio implementation
          (Transformation|AttachComponentToComponent _jumpboots_jump _meshpart "None" "SnapToTarget" "KeepRelative" "KeepRelative" true)
          (Audio|Components|Audio|SetSound _jumpboots_jump (Utilities|Array|Get(acopy) (Variables|Default|GetJumpSounds) _returnvalue))
          (Audio|Components|Audio|Play _jumpboots_jump 0.0))))))

(event EventOnLoopingStart(EitherVFX-preferthisversion)
  (params (MyTarget :type Actor%20Object%20Reference) (Parameters :type Gameplay%20Cue%20Parameters%20Structure%20%28by%20ref%29) (ParticleComponents :type Array%20of%20FXSystem%20Component%20Object%20References) (AudioComponents :type Array%20of%20Audio%20Component%20Object%20References)))

(event EventOnRemoval(EitherVFX-preferthisversion)
  (params (MyTarget :type Actor%20Object%20Reference) (Parameters :type Gameplay%20Cue%20Parameters%20Structure%20%28by%20ref%29) (ParticleComponents :type Array%20of%20FXSystem%20Component%20Object%20References) (AudioComponents :type Array%20of%20Audio%20Component%20Object%20References) (BurstCameraShakeInstance :type Camera%20Shake%20Base%20Object%20Reference) (BurstDecalInstance :type Decal%20Actor%20Object%20Reference))
  (bind (_normalizedmagnitude _rawmagnitude) (Ability|GameplayCue|BreakGameplayCueParameters Parameters))
  (bind _returnvalue (Math|Float|Truncate _rawmagnitude))
  (bind _returnvalue_2 (Niagara|SpawnSystematLocation 0 "/PrimalGameplay/Items/JumpBoots/Assets/Effects/NS_JumpBoots_OnLand.NS_JumpBoots_OnLand" (Transformation|GetActorLocation MyTarget)))
  (Effects|Components|ParticleSystem|SetIntParameter _returnvalue_2 "StageNum" _returnvalue)
  (Niagara|SetNiagaraVariable(Bool) _returnvalue_2 "bValidLanding" false)
  (Audio|SpawnSoundAttached (Utilities|Array|Get(acopy) (Variables|Default|GetLandSounds) _returnvalue) (Variables|Default|GetJumpBoots_Jump)))

(fn ConstructionScript)
