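; Gameplay cue notify that plays the landing effects for the exotic Jump Boots on a player pawn.
; OnBurst queues PlayBurstFX until the Gameplay cosmetic part is applied, passing the cue's raw magnitude as a string.
; For stages below 2, it attaches emissive Niagara to that part, timed by the jump reset value.
; It then waits for the Body part to spawn the landing effect and sets landing material parameters on the Gameplay mesh.
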(event Custom|PlayBurstFX
  (params (Pawn :type Fort%20Player%20Pawn%20Object%20Reference) (Part :type Custom%20Character%20Part%20Object%20Reference) (MeshPart :type Skeletal%20Mesh%20Component%20Object%20Reference) (PartType :type EFortCustomPartType%20Enum :default "Head") (AnyStruct :type Instanced%20Struct%20Structure%20%28by%20ref%29))
  (Utilities|IsValid MeshPart
    (:Is%20Valid
      (bind (_objectarray _stringarray) (Utilities|UnpackDatawhenCosmeticPartApplied AnyStruct)
        (:Valid
          (if (< (Math|Float|Truncate (Utilities|String|StringToFloat (Utilities|Array|Get(aref) _stringarray))) 2)
            (bind _returnvalue (Niagara|SpawnSystemAttached "/PrimalGameplay/Items/JumpBoots/Assets/Effects/NS_JumpBoots_Emissive.NS_JumpBoots_Emissive" MeshPart "root" "0, 0, 0" "0, 0, 0" "SnapToTarget" :!pinned true)) ; Spawn emissive glow
            (Effects|Components|ParticleSystem|SetIntParameter _returnvalue "StageNum" (Math|Float|Truncate (Utilities|String|StringToFloat (Utilities|Array|Get(aref) _stringarray))))
            (Niagara|SetNiagaraVariable(Float) _returnvalue "Delay" (ScalableFloat|GetValueatLevel (Variables|Default|GetJumpResetTime)))
            (bind (_pawn _part _meshpart) (Cosmetics|WaitForCosmeticSpecificPartApplied 0 Pawn true "Body")
              (:OnPartApplied
                (bind _returnvalue_2 (Niagara|SpawnSystemAttached "/PrimalGameplay/Items/JumpBoots/Assets/Effects/NS_JumpBoots_OnLand.NS_JumpBoots_OnLand" _meshpart "root" "0, 0, 0" "0, 0, 0" "SnapToTarget" :!pinned true)) ; Spawn ground effect
                (Effects|Components|ParticleSystem|SetIntParameter _returnvalue_2 "StageNum" (Math|Float|Truncate (Utilities|String|StringToFloat (Utilities|Array|Get(aref) _stringarray))))
                (Niagara|SetNiagaraVariable(Bool) _returnvalue_2 "bValidLanding" true)
                ;[1 1 1] Material Emissives
                (Rendering|Material|SetScalarParameterValueonMaterials MeshPart "IsLanding" 1.0)
                (Rendering|Material|SetScalarParameterValueonMaterials MeshPart "GameTimeFromBP" (Utilities|Time|GetGameTimeinSeconds MeshPart))))))))) ; Because the character part might not be visible we need to check before we attach a system to it
)

(fn OnBurst(EitherVFX-preferthisversion)
  (params (MyTarget :type Actor%20Object%20Reference) (Parameters :type Gameplay%20Cue%20Parameters%20Structure%20%28by%20ref%29) (ParticleComponents :type Array%20of%20FXSystem%20Component%20Object%20References) (AudioComponents :type Array%20of%20Audio%20Component%20Object%20References) (BurstCameraShakeInstance :type Camera%20Shake%20Base%20Object%20Reference) (BurstDecalInstance :type Decal%20Actor%20Object%20Reference))
  (bind (_normalizedmagnitude _rawmagnitude) (Ability|GameplayCue|BreakGameplayCueParameters Parameters))
  (bind _asplayer_pawn_athena (Utilities|Casting|CastToPlayerPawn_Athena MyTarget)
    (:then
      (bind _delegatehandlecontroller_out (Default|QueueWorkwhenSpecificCosmeticPartApplied :Target _asplayer_pawn_athena :Delegate (EventDispatchers|CreateEvent :!event PlayBurstFX) :PartType "Gameplay" :StringArray (Utilities|Array|MakeArray (Utilities|String|ToString(Float) _rawmagnitude))))
      (LocalVariables|SetHandle%20Controller%20Out _delegatehandlecontroller_out)
      (return))
    (:CastFailed
      (return))))

(fn ConstructionScript)
